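# Physics asset for the M_MED_Aztec backpack cosmetic, giving collision bodies and joints to its dangling earrings and tongue.
# A two-capsule backpack_root body anchors both earring capsules, whose joints lock twist, and the first tongue segment.
# The tongue chains four boxes, then forks at the fourth into two sibling capsule tips, with small limit angles set on each joint.
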PhysicsAsset /BRCosmetics/Accessories/FORT_Backpacks/M_MED_Aztec_Backpack/Meshes/M_MED_Aztec_Backpack_Physics
bodies {
 backpack_root [
  {shapeType=Capsule, length=15.2605, radius=3.5313, center=<-14.5876 -7.6577 0.0025>, rotation=<rot 0.0052 -89.9999 0.0482>}
  {shapeType=Capsule, length=15, radius=15, center=<-3.9688 3.5306 0>}
 ]
 dyn_earring_l [{shapeType=Capsule, length=7.6211, radius=2.1551, center=<5.5744 -0.651 0.0197>, rotation=<rot 89.7154 179.9999 180>}]
 dyn_earring_r [{shapeType=Capsule, length=7.6211, radius=2.1551, center=<5.5744 -0.651 0.0197>, rotation=<rot 89.7154 179.9999 180>}]
 dyn_tongue_01 [{shapeType=Box, extentX=2.7617, extentY=0.5193, extentZ=3.3656, center=<1.5199 -0.1914 0>, rotation=<rot 0 -4.8996 0>}]
 dyn_tongue_02 [{shapeType=Box, extentX=3.2343, extentY=0.5193, extentZ=3.3656, center=<1.8548 -0.0569 0>, rotation=<rot 0 -4.0474 0>}]
 dyn_tongue_03 [{shapeType=Box, extentX=3.737, extentY=0.5193, extentZ=3.3656, center=<2.0928 -0.1311 0>, rotation=<rot 0 -8.9361 0>}]
 dyn_tongue_04 [{shapeType=Box, extentX=3.2092, extentY=0.5193, extentZ=5.8205, center=<1.9591 0.0495 0>, rotation=<rot 0 1.4476 0>}]
 dyn_tongue_05 [{shapeType=Capsule, length=2.6936, radius=1.3966, center=<0.9433 -0.3291 0.0175>, rotation=<rot -89.8032 -0.001 0.001>}]
 dyn_tongue_06 [{shapeType=Capsule, length=2.6936, radius=1.3966, center=<1.1416 -0.3295 0.0514>, rotation=<rot 87.4025 179.9992 179.9994>}]
}
constraints [
 {bone1Name=dyn_earring_l, bone2Name=backpack_root, swing1LimitDegrees=10, swing2LimitDegrees=35, twistLimitDegrees=45, twistMotion=Locked}
 {bone1Name=dyn_earring_r, bone2Name=backpack_root, swing1LimitDegrees=10, swing2LimitDegrees=35, twistLimitDegrees=45, twistMotion=Locked}
 {bone1Name=dyn_tongue_01, bone2Name=backpack_root, swing1LimitDegrees=10, swing2LimitDegrees=10, twistLimitDegrees=2}
 {bone1Name=dyn_tongue_02, bone2Name=dyn_tongue_01, swing1LimitDegrees=10, swing2LimitDegrees=10, twistLimitDegrees=2}
 {bone1Name=dyn_tongue_03, bone2Name=dyn_tongue_02, swing1LimitDegrees=10, swing2LimitDegrees=10, twistLimitDegrees=2}
 {bone1Name=dyn_tongue_04, bone2Name=dyn_tongue_03, swing1LimitDegrees=10, swing2LimitDegrees=10, twistLimitDegrees=2}
 {bone1Name=dyn_tongue_05, bone2Name=dyn_tongue_04, swing1LimitDegrees=5, swing2LimitDegrees=10, twistLimitDegrees=2}
 {bone1Name=dyn_tongue_06, bone2Name=dyn_tongue_04, swing1LimitDegrees=5, swing2LimitDegrees=10, twistLimitDegrees=2}
]
